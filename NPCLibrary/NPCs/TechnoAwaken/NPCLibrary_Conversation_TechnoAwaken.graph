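# Conversation database for the TechnoAwaken NPC, providing its beat-and-rhythm themed greeting dialogue and routing into shared conversation branches.
# The entry point leads to a single Speech node with a default line plus contextual lines, including a prioritized first-ever-meeting greeting and others gated by NPC match conversation count checks.
# The speech then branches through link nodes to the remote NPC library quests, basic services and Ch5S1 D3 Q8 entry tags.
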ConversationDatabase /NPCLibrary/NPCs/TechnoAwaken/NPCLibrary_Conversation_TechnoAwaken
entry_points [{tag=Conversation.Entry.TechnoAwaken, destinations [#1]}]
nodes [
 ConversationEntryPointNode_7 {
  id=#1
  class=ConversationEntryPointNode
  outputs [#2]
  properties {entryTag {tagName=Conversation.Entry.TechnoAwaken}}
 }
 Speech_C_4 {
  id=#2
  class=Speech_C
  outputs [#3, #4, #5]
  properties {
   nodeName=""
   generalConfig {
    defaultMessage=The Island's got its own rhythm. Just gotta know how to listen.
    contextualMessages [
     {
      message="Hey, you feel that beat? Listen."
      contextRequirements [{requirement {class=FortControllerRequirement_IsFirstEverConversationWithNPC}}]
      priority {curve {rowName=Tandem.Speech.Priority.FirstEverInteraction}}
     }
     {
      message=The Island's got its own rhythm. Just gotta know how to listen.
      contextRequirements [{requirement {class=FortControllerRequirement_CheckNPCMatchConversationCount, properties {nPCUniqueIdTag {tagName=AISpawnerData.Type.Tandem.TechnoAwaken}}}}]
      priority {curve {rowName=None}}
     }
     {
      message="Everything's got a beat. This Island, me, even you."
      contextRequirements [{requirement {class=FortControllerRequirement_CheckNPCMatchConversationCount, properties {nPCUniqueIdTag {tagName=AISpawnerData.Type.Tandem.TechnoAwaken}}}}]
      priority {curve {rowName=None}}
     }
     {
      message=Stayin' safe out there? The beat's goin' wild today.
      contextRequirements [{requirement {class=FortControllerRequirement_CheckNPCMatchConversationCount, properties {nPCUniqueIdTag {tagName=AISpawnerData.Type.Tandem.TechnoAwaken}}}}]
      priority {curve {rowName=None}}
     }
     {
      message="The Society may be falling, but somethin' else will pick up the beat. Trust me."
      contextRequirements [{requirement {class=FortControllerRequirement_CheckNPCMatchConversationCount, properties {nPCUniqueIdTag {tagName=AISpawnerData.Type.Tandem.TechnoAwaken}}}}]
      priority {curve {rowName=None}}
     }
    ]
   }
  }
 }
 ConversationLinkNode_2 {id=#5, class=ConversationLinkNode, properties {remoteEntryTag {tagName=Conversation.Entry.Ch5S1.D3.Q8}}}
 ConversationLinkNode_0 {id=#4, class=ConversationLinkNode, properties {remoteEntryTag {tagName=Conversation.Entry.BasicServices}}}
 ConversationLinkNode_1 {id=#3, class=ConversationLinkNode, properties {remoteEntryTag {tagName=Conversation.Entry.NPCLibrary.Quests}}}
]
